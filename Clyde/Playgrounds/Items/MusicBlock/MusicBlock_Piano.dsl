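; Piano key music block that plays its note and glows when damaged, overlapped or hit by a music sequencer.
; Server-side damage and overlap force building health and fire Hit Key behind a DoOnce; cosmetic-capable instances play audio and a glow timeline.
; Glow borrows a dynamic material from FortMaterialCache, skipping sound and animation if none is available, and releases it afterward.
; Key index selects the label's UV offset and the sound parameter.
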(event EventBeginPlay
  (Networking|CanExecuteCosmeticEvents
    (:True
      ;[1 1 1] Added for papaya event.
      (CallFunction|UpdateBody)
      (bind _audioattenuationsettings (Variables|PianoKey|Sounds|GetAudioAttenuationSettings)
        (:then
          (Variables|/Script/Engine.AudioComponent|SetAttenuationSettings _audioattenuationsettings (Variables|Default|GetAudio)))
        (:else)))))

(event Custom|Hit%20Key
  (Networking|CanExecuteCosmeticEvents
    (:True
      (CallFunction|SwitchToDyamicMaterial)
      (CallFunction|AnimateDynamicMaterial))))

(event EventOnHitByMusicSequencer
  (params (InstigatedBy :type Controller%20Object%20Reference))
  (Networking|CanExecuteCosmeticEvents
    (:True
      (CallFunction|SwitchToDyamicMaterial)
      (CallFunction|AnimateDynamicMaterial))))

(event OnComponentBeginOverlap(AudioTrigger)
  (params (OverlappedComponent :type Primitive%20Component%20Object%20Reference) (OtherActor :type Actor%20Object%20Reference) (OtherComp :type Primitive%20Component%20Object%20Reference) (OtherBodyIndex :type Integer) (bFromSweep :type Boolean) (SweepResult :type Hit%20Result%20Structure%20%28by%20ref%29))
  (Networking|SwitchHasAuthority
    (:Authority
      (Gameplay|ForceBuildingHealth 1e+02)
      (Utilities|FlowControl|DoOnce false
        (:Completed
          (CallFunction|HitKey)
          (Utilities|FlowControl|Delay 0.2))))))

(event EventOnDamageServer
  (params (Damage :type Float%20%28single-precision%29) (DamageTags :type Gameplay%20Tag%20Container%20Structure%20%28by%20ref%29) (Momentum :type Vector :default "0, 0, 0") (HitInfo :type Hit%20Result%20Structure%20%28by%20ref%29) (InstigatedBy :type Controller%20Object%20Reference) (DamageCauser :type Actor%20Object%20Reference) (EffectContext :type Gameplay%20Effect%20Context%20Handle%20Structure))
  (Networking|SwitchHasAuthority
    (:Authority
      (Gameplay|ForceBuildingHealth 1e+02)
      (Utilities|FlowControl|DoOnce false
        (:Completed
          (CallFunction|HitKey)
          (Utilities|FlowControl|Delay 0.2))))))

(event Custom|Switch%20to%20Dyamic%20Material
  (Networking|CanExecuteCosmeticEvents
    (:True
      (Variables|Default|GetMaterial%20Instance%20Dynamic
        (:then
          _)
        (:else
          (bind _material_cache (Variables|Default|GetMaterial%20Cache)
            (:then
              (bind (_outmaterialinstance _returnvalue) (Class|FortMaterialCache|ReserveMaterial _material_cache))
              (Variables|Default|SetMaterial%20Instance%20Dynamic _outmaterialinstance)
              (if _returnvalue
                _))
            (:else))))
      (Rendering|Material|SetMaterial (Variables|Default|GetMesh) 0 (Variables|Default|GetMaterial%20Instance%20Dynamic)))))

(event Custom|Switch%20to%20Constant%20Material
  (Networking|CanExecuteCosmeticEvents
    (:True
      (Rendering|Material|SetMaterial (Variables|Default|GetMesh) 0 (Utilities|GetMaterialInstanceConstant))
      (bind _material_cache (Variables|Default|GetMaterial%20Cache)
        (:then
          (Class|FortMaterialCache|ReleaseMaterial _material_cache (Variables|Default|GetMaterial%20Instance%20Dynamic)))
        (:else)))))

(event Custom|Animate%20Dynamic%20Material
  (bind _material_instance_dynamic (Variables|Default|GetMaterial%20Instance%20Dynamic)
    (:then
      (Audio|Components|Audio|Play (Variables|Default|GetAudio) 0.0)
      (bind (_direction _glow) (Timeline|Timeline_0 PlayFromStart)
        (:Update
          (Rendering|Material|SetScalarParameterValue _material_instance_dynamic "Glow" _glow :!class /Script/Engine.MaterialInstanceDynamic))
        (:Finished
          (Rendering|Material|SetScalarParameterValue _material_instance_dynamic "Glow" _glow :!class /Script/Engine.MaterialInstanceDynamic)
          (CallFunction|SwitchToConstantMaterial))))
    (:else)) ; Skip animating if we failed to get a dynamic\nmaterial instance (too may things animating)
)

(event EventEndPlay
  (params (EndPlayReason :type EEndPlayReason%20Enum :default "Destroyed"))
  (CallFunction|SwitchToConstantMaterial) ; Make sure to release the cached material
)

(fn GetMusicSequencerHitSound
  (params (InstigatedBy :type Controller%20Object%20Reference))
  (return 0 0 0 0))

(fn GetMusicSequencerHitSoundLocation
  (params (InstigatedBy :type Controller%20Object%20Reference))
  (return "0, 0, 0"))

(fn GetOverrideMeshMaterial
  (params (MeshComp :type Static%20Mesh%20Component%20Object%20Reference) (MatIdx :type Integer))
  (if (Utilities|Equal(Object) MeshComp (Variables|Default|GetMesh))
    (switch int MatIdx
      (:Default
        (return 0))
      (:0
        (bind _material_instance_dynamic (Variables|Default|GetMaterial%20Instance%20Dynamic)
          (:then
            (return _material_instance_dynamic))
          (:else
            (return (Utilities|GetMaterialInstanceConstant)))))
      (:1
        (return (Variables|PianoKey|Materials|GetText%20Material%20Instance))))
    (else
      (return 0))))

(fn OnRep_bUseAlternateApperance
  (CallFunction|UpdateBody)
  (CallFunction|UpdateText))

(fn OnRep_Index
  (CallFunction|UpdateText)
  (CallFunction|UpdateAudio))

(fn UpdateAudio
  (Audio|Components|Audio|SetSound (Variables|Default|GetAudio) (select (TouchPlatform|IsMobileGame) (Variables|PianoKey|Sounds|GetSound%20Mobile) (Variables|PianoKey|Sounds|GetSound%20Default)))
  (Audio|Parameter|SetIntegerParameter (Variables|Default|GetAudio) "KeysIndex" (Variables|PianoKey|GetIndex) :!class /Script/Engine.AudioComponent))

(fn UpdateBody
  (bind _returnvalue (Utilities|GetMaterialInstanceConstant))
  (Rendering|Material|SetMaterial (Variables|Default|GetMesh) 0 _returnvalue)
  (bind _returnvalue_2 (Class|FortMaterialCacheManager|GetCache (Variables|Default|GetMaterial%20Cache%20Settings) _returnvalue))
  (Variables|Default|SetMaterial%20Cache _returnvalue_2))

(fn UpdateText
  (bind _text_material_instance (Variables|PianoKey|Materials|GetText%20Material%20Instance))
  (Rendering|Material|SetVectorParameterValue _text_material_instance "UV Offset" (Math|Color|MakeColor (* (Math|Integer|%25%28Integer%29 (Variables|PianoKey|GetIndex) 5) 0.2) (* (/ (Variables|PianoKey|GetIndex) 5) 0.2) 0.0 0.0) :!class /Script/Engine.MaterialInstanceDynamic)
  (Rendering|Material|SetVectorParameterValue _text_material_instance "letter_tint" (Utilities|GetTextColor) :!class /Script/Engine.MaterialInstanceDynamic))

(fn ConstructionScript
  (Utilities|BuildMaterial (Variables|Default|GetMesh) 1 0 (Variables|PianoKey|Materials|GetText%20Material%20Instance)) ; Rebuild the material for the text
  (CallFunction|UpdateBody)
  (CallFunction|UpdateText)
  (CallFunction|UpdateAudio))
